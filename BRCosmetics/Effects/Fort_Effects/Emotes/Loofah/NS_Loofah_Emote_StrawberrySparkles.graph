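# Niagara system for the Loofah emote's strawberry watering-can prop, layering timed one-shot bursts of sparkles, flares, coronas, splashes, smoke puffs and dirt chunks.
# Early bursts fire at the prop's FX_Socket, while the closing bursts fire at the strawberry_02 bone, both found by skeletal-mesh sampling.
# Delayed instantaneous bursts time the stages; some flares add light renderers, and sparkles flicker through color curves and drift on curl noise.
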NiagaraSystem /BRCosmetics/Effects/Fort_Effects/Emotes/Loofah/NS_Loofah_Emote_StrawberrySparkles
dependencies {
 usedDataInterfaces [
  &NiagaraDataInterfaceSkeletalMesh
  &NiagaraDataInterfaceVectorCurve
  &NiagaraDataInterfaceCurve
  &NiagaraDataInterfaceVector2DCurve
  &NiagaraDataInterfaceColorCurve
 ]
 usedDynamicInputs []
 usedModules [&SystemState]
}
emitters {
 Corona {
  emitterUpdateScript [
   &EmitterState(Life Cycle Mode=::NewEnumerator1, Loop Behavior=::NewEnumerator1, Loop Duration=10, UseLoopDelay=false)
   &SpawnBurst_Instantaneous(Spawn Count=1, Spawn Time=2.7, Use Spawn Probability=false)
  ]
  particleSpawnScript [
   /Niagara/Modules/Spawn/Initialization/InitializeParticle(
    Color=<rgba 1 0.307 0.8549 0.5>
    Lifetime=&UniformRangedFloat(Minimum=0.2, Maximum=0.2)
    Mesh Scale=<1.5 1.5 1.5>
    Module.Write Color=true
    Module.Write Lifetime=true
    Module.Write Mass=true
    Module.Write Material Random=true
    Module.Write Position=false
    Module.Write Scale=false
    Module.Write SpriteRotation=true
    Module.Write SpriteSize=true
    Position=&SimulationPosition()
    Sprite Rotation=&UniformRangedFloat(Minimum=0, Maximum=1)
    Sprite Size=<120 120>
   )
   &SampleSkeletalMeshSkeleton(
    Mesh=DataInterface() {propertyValues {
      DefaultMesh=null
      ExcludeBoneName=root
      OptionalComponentClass=null
      PreviewMesh=/BRCosmetics/Animation/Props/Watering_Can_Strawberry/Meshes/Watering_Can_Strawberry_Gadget
      SkinningMode=SkinOnTheFly
      SoftSourceActor=null
      SourceMode=Default
      UvSetIndex=0
      WholeMeshLOD=-1
      bExcludeBone=false
      bReadDeformedGeometry=true
      bRequireCurrentFrameData=true
      ComponentTags []
      FilteredBones []
      FilteredSockets [FX_Socket]
      MeshUserParameter {parameter {name=None, typeDefHandle {registeredTypeIndex=1}, varData []}}
      SamplingRegions []
     }}
    Spawn Mode:ENiagaraSkeletalSpawning=::NewEnumerator1
   )
   /Niagara/Modules/Spawn/Location/SkeletalMeshSkeletonLocation(Module.Inherit Velocity=false, Module.Use Sampled Bone as Ribbon ID=true)
   /Niagara/Modules/Spawn/Location/SphereLocation(MaskSpawnGroup=false, Module.Override Randomness=false, Offset=<-20 100 0>, Sphere Radius=0, Surface Only=false) {enabled=false}
   &SystemLocation(MaskSpawnGroup=false, Offset=<-20 100 0>) {enabled=false}
  ]
  particleUpdateScript [
   &ParticleState()
   &Drag(Drag=3, Use Linear Drag=true, Use Rotational Drag=false)
   &SolveForcesAndVelocity(Clamp Velocity=false, Limit Acceleration=false)
   &ScaleSpriteSize(Scale Factor=&Vector2DFromFloat(Value=&FloatFromCurve(FloatCurve=DataInterface() {propertyValues {Curve {keys [{arriveTangent=1.2387, interpMode=RCIM_Cubic, leaveTangent=1.2387, tangentMode=RCTM_User, time=0, value=0.5}, <at 1 1 RCIM_Cubic>]}}})))
   &CameraOffset(Camera Offset Amount=20)
   &DynamicMaterialParameters(
    Index 0 Param 1=&FloatFromCurve(FloatCurve=DataInterface() {propertyValues {Curve {keys [<at -0.0046 1.0113>,<at 1 0.5>]}}})
    Index 0 Param 2=1
    Index 0 Param 3=1
    Index 0 Param 4=1
    Module.Param0WriteEnabled=true
    Module.Param1WriteEnabled=true
    Module.Param2WriteEnabled=true
    Module.Param3WriteEnabled=true
   )
  ]
  rendererClasses [&NiagaraSpriteRendererProperties]
  renderers [{rendererClass=&NiagaraSpriteRendererProperties}]
 }
 Corona_End {
  emitterUpdateScript [
   &EmitterState(Life Cycle Mode=::NewEnumerator1, Loop Behavior=::NewEnumerator1, Loop Duration=10, UseLoopDelay=false)
   &SpawnBurst_Instantaneous(Spawn Count=1, Spawn Time=8.05, Use Spawn Probability=false)
  ]
  particleSpawnScript [
   &InitializeParticle(
    Color=<rgba 0.1172 0.3901 0.9 0.3>
    Lifetime=&UniformRangedFloat(Minimum=0.2, Maximum=0.2)
    Mesh Scale=<1.5 1.5 1.5>
    Module.Write Color=true
    Module.Write Lifetime=true
    Module.Write Mass=true
    Module.Write Material Random=true
    Module.Write Position=false
    Module.Write Scale=false
    Module.Write SpriteRotation=true
    Module.Write SpriteSize=true
    Position=&SimulationPosition()
    Sprite Rotation=&UniformRangedFloat(Minimum=0, Maximum=1)
    Sprite Size=<100 100>
   )
   &SampleSkeletalMeshSkeleton(Mesh=DataInterface() {propertyValues {
      DefaultMesh=null
      ExcludeBoneName=root
      OptionalComponentClass=null
      PreviewMesh=&Watering_Can_Strawberry_Gadget
      SkinningMode=SkinOnTheFly
      SoftSourceActor=null
      SourceMode=Default
      UvSetIndex=0
      WholeMeshLOD=-1
      bExcludeBone=false
      bReadDeformedGeometry=true
      bRequireCurrentFrameData=true
      ComponentTags []
      FilteredBones [strawberry_02]
      FilteredSockets []
      MeshUserParameter {parameter {name=None, typeDefHandle {registeredTypeIndex=1}, varData []}}
      SamplingRegions []
     }})
   &SkeletalMeshSkeletonLocation(Module.Inherit Velocity=false, Module.Use Sampled Bone as Ribbon ID=true)
   &SphereLocation(MaskSpawnGroup=false, Module.Override Randomness=false, Offset=<-20 100 0>, Sphere Radius=0, Surface Only=false) {enabled=false}
   &SystemLocation(MaskSpawnGroup=false, Offset=<-20 100 0>) {enabled=false}
  ]
  particleUpdateScript [
   &ParticleState()
   &Drag(Drag=3, Use Linear Drag=true, Use Rotational Drag=false)
   &SolveForcesAndVelocity(Clamp Velocity=false, Limit Acceleration=false)
   &ScaleSpriteSize(Scale Factor=&Vector2DFromFloat(Value=&FloatFromCurve(FloatCurve=DataInterface() {propertyValues {Curve {keys [{arriveTangent=1.2387, interpMode=RCIM_Cubic, leaveTangent=1.2387, tangentMode=RCTM_User, time=0, value=0.5}, <at 1 1 RCIM_Cubic>]}}})))
   &CameraOffset(Camera Offset Amount=20)
   &DynamicMaterialParameters(
    Index 0 Param 1=&FloatFromCurve(FloatCurve=DataInterface() {propertyValues {Curve {keys [<at -0.0046 1.0113>,<at 1 0.5>]}}})
    Index 0 Param 2=1
    Index 0 Param 3=1
    Index 0 Param 4=1
    Module.Param0WriteEnabled=true
    Module.Param1WriteEnabled=true
    Module.Param2WriteEnabled=true
    Module.Param3WriteEnabled=true
   )
  ]
  rendererClasses [&NiagaraSpriteRendererProperties]
  renderers [{rendererClass=&NiagaraSpriteRendererProperties}]
 }
 Dirt_Chunks {
  emitterUpdateScript [
   &EmitterState(Life Cycle Mode=::NewEnumerator1, Loop Behavior=::NewEnumerator1, Loop Duration=10, UseLoopDelay=false)
   &SpawnBurst_Instantaneous(Spawn Count=20, Spawn Time=2.7, Use Spawn Probability=false)
   &SpawnRate(SpawnRate=50, Use Spawn Probability=false) {enabled=false}
  ]
  particleSpawnScript [
   &InitializeParticle(
    Color=<rgba 0.1276 0.1276 0.1276 1>
    Lifetime=&UniformRangedFloat(Minimum=1, Maximum=1)
    Mesh Scale=<1.5 1.5 1.5>
    Module.Write Color=true
    Module.Write Lifetime=true
    Module.Write Mass=false
    Module.Write Material Random=true
    Module.Write Position=false
    Module.Write Scale=false
    Module.Write SpriteRotation=true
    Module.Write SpriteSize=true
    Position=&SimulationPosition()
    Sprite Rotation=&UniformRangedFloat(Minimum=0, Maximum=1)
    Sprite Size=&UniformRangedVector2D(
     Minimum=&Vector2DFromFloat(Value=2.5)
     Maximum=&Vector2DFromFloat(Value=5)
    )
   )
   &SampleSkeletalMeshSkeleton(
    Mesh=DataInterface() {propertyValues {
      DefaultMesh=null
      ExcludeBoneName=root
      OptionalComponentClass=null
      PreviewMesh=&Watering_Can_Strawberry_Gadget
      SkinningMode=SkinOnTheFly
      SoftSourceActor=null
      SourceMode=Default
      UvSetIndex=0
      WholeMeshLOD=-1
      bExcludeBone=false
      bReadDeformedGeometry=true
      bRequireCurrentFrameData=true
      ComponentTags []
      FilteredBones []
      FilteredSockets [FX_Socket]
      MeshUserParameter {parameter {name=None, typeDefHandle {registeredTypeIndex=1}, varData []}}
      SamplingRegions []
     }}
    Spawn Mode:ENiagaraSkeletalSpawning=::NewEnumerator1
   )
   &SkeletalMeshSkeletonLocation(Module.Inherit Velocity=false, Module.Use Sampled Bone as Ribbon ID=true)
   &SphereLocation(MaskSpawnGroup=false, Module.Override Randomness=false, Sphere Radius=0, Surface Only=false)
   &AddVelocity(Velocity=&UniformRangedVector(Minimum=<200 200 400>, Maximum=<-200 -200 200>))
   /Niagara/Modules/Update/SubUV/SubUVAnimation(Number Of Frames=1, SubUV Animation Mode=::NewEnumerator1) {enabled=false}
  ]
  particleUpdateScript [
   &ParticleState()
   &Drag(Use Linear Drag=true, Use Rotational Drag=false)
   &GravityForce()
   &SolveForcesAndVelocity(Clamp Velocity=false, Limit Acceleration=false)
   &ScaleSpriteSize(Scale Factor=&Vector2DFromFloat(Value=&FloatFromCurve(FloatCurve=DataInterface() {propertyValues {Curve {keys [<at 0 0.5 RCIM_Cubic>,<at 1 1 RCIM_Cubic>]}}}))) {enabled=false}
  ]
  rendererClasses [&NiagaraSpriteRendererProperties]
  renderers [
   {rendererClass=&NiagaraSpriteRendererProperties}
   {rendererClass=&NiagaraSpriteRendererProperties, rendererIndex=1}
  ]
 }
 Flare {
  emitterUpdateScript [
   &EmitterState(Life Cycle Mode=::NewEnumerator1, Loop Behavior=::NewEnumerator1, Loop Duration=10, Scalability Mode=::NewEnumerator1, UseLoopDelay=false)
   &SpawnBurst_Instantaneous(Spawn Count=1, Spawn Time=2.7, Use Spawn Probability=false)
  ]
  particleSpawnScript [
   &InitializeParticle(
    Color=<rgba 50 15.4499 23.2308 0.3>
    Lifetime=&UniformRangedFloat(Minimum=0.1, Maximum=0.1)
    Mesh Scale=<1.5 1.5 1.5>
    Module.Write Color=true
    Module.Write Lifetime=true
    Module.Write Mass=true
    Module.Write Material Random=true
    Module.Write Position=false
    Module.Write Scale=false
    Module.Write SpriteRotation=true
    Module.Write SpriteSize=true
    Position=&SimulationPosition()
    Sprite Rotation=&UniformRangedFloat(Minimum=0, Maximum=1)
    Sprite Size=&UniformRangedVector2D(
     Minimum=&Vector2DFromFloat(Value=75)
     Maximum=&Vector2DFromFloat(Value=100)
    )
   )
   &SampleSkeletalMeshSkeleton(
    Mesh=DataInterface() {propertyValues {
      DefaultMesh=null
      ExcludeBoneName=root
      OptionalComponentClass=null
      PreviewMesh=&Watering_Can_Strawberry_Gadget
      SkinningMode=SkinOnTheFly
      SoftSourceActor=null
      SourceMode=Default
      UvSetIndex=0
      WholeMeshLOD=-1
      bExcludeBone=false
      bReadDeformedGeometry=true
      bRequireCurrentFrameData=true
      ComponentTags []
      FilteredBones []
      FilteredSockets [FX_Socket]
      MeshUserParameter {parameter {name=None, typeDefHandle {registeredTypeIndex=1}, varData []}}
      SamplingRegions []
     }}
    Spawn Mode:ENiagaraSkeletalSpawning=::NewEnumerator1
   )
   &SkeletalMeshSkeletonLocation(Module.Inherit Velocity=false, Module.Use Sampled Bone as Ribbon ID=true)
   &SphereLocation(MaskSpawnGroup=false, Module.Override Randomness=false, Offset=<-20 100 0>, Sphere Radius=0, Surface Only=false) {enabled=false}
   &SystemLocation(MaskSpawnGroup=false) {enabled=false}
   &FN_Light_v2(Light Radius=100, Position Offset:Vector3f=<0 0 50>, Scale Color=1)
   /SetVariables(Particles.Light Color:LinearColor=<rgba 50 15.4499 23.2308 1>) {isSetParameters=true}
  ]
  particleUpdateScript [
   &ParticleState()
   &Drag(Drag=3, Use Linear Drag=true, Use Rotational Drag=false)
   &SolveForcesAndVelocity(Clamp Velocity=false, Limit Acceleration=false)
   &ScaleSpriteSize(Scale Factor=&Vector2DFromFloat(Value=&FloatFromCurve(FloatCurve=DataInterface() {propertyValues {Curve {keys [{arriveTangent=1.2387, interpMode=RCIM_Cubic, leaveTangent=1.2387, tangentMode=RCTM_User, time=0, value=0.5}, <at 1 1 RCIM_Cubic>]}}})))
   &CameraOffset(Camera Offset Amount=20)
  ]
  rendererClasses [&NiagaraSpriteRendererProperties, &NiagaraLightRendererProperties]
  renderers [
   {rendererClass=&NiagaraSpriteRendererProperties}
   {rendererClass=&NiagaraLightRendererProperties, rendererIndex=1}
  ]
 }
 Flare_Big_End {
  emitterUpdateScript [
   &EmitterState(Life Cycle Mode=::NewEnumerator1, Loop Behavior=::NewEnumerator1, Loop Duration=15, UseLoopDelay=false)
   &SpawnBurst_Instantaneous(Spawn Count=1, Spawn Time=8.05, Use Spawn Probability=false)
  ]
  particleSpawnScript [
   &InitializeParticle(
    Color=<rgba 50 15.4499 23.2309 0.5>
    Lifetime=&UniformRangedFloat(Minimum=0.1, Maximum=0.1)
    Mesh Scale=<1.5 1.5 1.5>
    Module.Write Color=true
    Module.Write Lifetime=true
    Module.Write Mass=true
    Module.Write Material Random=true
    Module.Write Position=false
    Module.Write Scale=false
    Module.Write SpriteRotation=true
    Module.Write SpriteSize=true
    Position=&SimulationPosition()
    Sprite Rotation=&UniformRangedFloat(Minimum=0, Maximum=1)
    Sprite Size=&UniformRangedVector2D(
     Minimum=&Vector2DFromFloat(Value=50)
     Maximum=&Vector2DFromFloat(Value=50)
    )
   )
   &SampleSkeletalMeshSkeleton(Mesh=DataInterface() {propertyValues {
      DefaultMesh=null
      ExcludeBoneName=root
      OptionalComponentClass=null
      PreviewMesh=&Watering_Can_Strawberry_Gadget
      SkinningMode=SkinOnTheFly
      SoftSourceActor=null
      SourceMode=Default
      UvSetIndex=0
      WholeMeshLOD=-1
      bExcludeBone=false
      bReadDeformedGeometry=true
      bRequireCurrentFrameData=true
      ComponentTags []
      FilteredBones [strawberry_02]
      FilteredSockets []
      MeshUserParameter {parameter {name=None, typeDefHandle {registeredTypeIndex=1}, varData []}}
      SamplingRegions []
     }})
   &SkeletalMeshSkeletonLocation(Module.Inherit Velocity=false, Module.Use Sampled Bone as Ribbon ID=true)
   &SphereLocation(MaskSpawnGroup=false, Module.Override Randomness=false, Offset=<-20 100 0>, Sphere Radius=0, Surface Only=false) {enabled=false}
   &SystemLocation(MaskSpawnGroup=false) {enabled=false}
   &CameraOffset(Camera Offset Amount=20)
  ]
  particleUpdateScript [
   &ParticleState()
   &Drag(Drag=3, Use Linear Drag=true, Use Rotational Drag=false)
   &SolveForcesAndVelocity(Clamp Velocity=false, Limit Acceleration=false)
   &ScaleSpriteSize(Scale Factor=&Vector2DFromFloat(Value=&FloatFromCurve(FloatCurve=DataInterface() {propertyValues {Curve {keys [{arriveTangent=1.2387, interpMode=RCIM_Cubic, leaveTangent=1.2387, tangentMode=RCTM_User, time=0, value=0.5}, <at 1 1 RCIM_Cubic>]}}})))
  ]
  rendererClasses [&NiagaraSpriteRendererProperties]
  renderers [{rendererClass=&NiagaraSpriteRendererProperties}]
 }
 Flare_Light_End {
  emitterUpdateScript [
   &EmitterState(
    Enable Distance Culling=true
    Life Cycle Mode=::NewEnumerator1
    Loop Behavior=::NewEnumerator1
    Loop Duration=10
    MaxDistance=3000
    Module.bUseMaxDistance=true
    Module.bUseMinDistance=false
    Scalability Mode=::NewEnumerator1
    UseLoopDelay=false
   )
   &SpawnBurst_Instantaneous(Spawn Count=1, Spawn Time=7.95, Use Spawn Probability=false)
   &SpawnBurst_Instantaneous(Spawn Count=1, Use Spawn Probability=false) {enabled=false, module=SpawnBurst_Instantaneous001}
  ]
  particleSpawnScript [
   &InitializeParticle(
    Color=<rgba 50 15.4499 23.2308 1>
    Lifetime=&UniformRangedFloat(Minimum=0.1, Maximum=0.1)
    Mesh Scale=<1.5 1.5 1.5>
    Module.Write Color=true
    Module.Write Lifetime=true
    Module.Write Mass=true
    Module.Write Material Random=true
    Module.Write Position=false
    Module.Write Scale=false
    Module.Write SpriteRotation=true
    Module.Write SpriteSize=true
    Position=&SimulationPosition()
    Sprite Rotation=&UniformRangedFloat(Minimum=0, Maximum=1)
    Sprite Size=<20 20>
   )
   &SampleSkeletalMeshSkeleton(Mesh=DataInterface() {propertyValues {
      DefaultMesh=null
      ExcludeBoneName=root
      OptionalComponentClass=null
      PreviewMesh=&Watering_Can_Strawberry_Gadget
      SkinningMode=SkinOnTheFly
      SoftSourceActor=null
      SourceMode=Default
      UvSetIndex=0
      WholeMeshLOD=-1
      bExcludeBone=false
      bReadDeformedGeometry=true
      bRequireCurrentFrameData=true
      ComponentTags []
      FilteredBones [strawberry_02]
      FilteredSockets []
      MeshUserParameter {parameter {name=None, typeDefHandle {registeredTypeIndex=1}, varData []}}
      SamplingRegions []
     }})
   &SkeletalMeshSkeletonLocation(Module.Inherit Velocity=false, Module.Use Sampled Bone as Ribbon ID=true)
   &SphereLocation(MaskSpawnGroup=false, Module.Override Randomness=false, Offset=<5 -15 0>, Sphere Radius=0, Surface Only=false)
   &SystemLocation(MaskSpawnGroup=false, Offset=<0 65 150>) {enabled=false}
   &FN_Light_v2(Light Radius=100, Position Offset:Vector3f=<0 -30 0>)
   /SetVariables(Particles.Light Color:LinearColor=<rgba 5 1.545 2.3231 1>) {isSetParameters=true, module=SetVariables_#1}
  ]
  particleUpdateScript [
   &ParticleState()
   &Drag(Drag=3, Use Linear Drag=true, Use Rotational Drag=false)
   &SolveForcesAndVelocity(Clamp Velocity=false, Limit Acceleration=false)
   &ScaleSpriteSize(Scale Factor=&Vector2DFromFloat(Value=&FloatFromCurve(FloatCurve=DataInterface() {propertyValues {Curve {keys [{arriveTangent=1.2387, interpMode=RCIM_Cubic, leaveTangent=1.2387, tangentMode=RCTM_User, time=0, value=0.5}, <at 1 1 RCIM_Cubic>]}}})))
   &CameraOffset(Camera Offset Amount=10)
  ]
  rendererClasses [&NiagaraSpriteRendererProperties, &NiagaraLightRendererProperties]
  renderers [
   {rendererClass=&NiagaraSpriteRendererProperties}
   {rendererClass=&NiagaraLightRendererProperties, rendererIndex=1}
  ]
 }
 Flare_Splash_End {
  emitterUpdateScript [
   &EmitterState(Life Cycle Mode=::NewEnumerator1, Loop Behavior=::NewEnumerator1, Loop Duration=10, UseLoopDelay=false)
   &SpawnBurst_Instantaneous(Spawn Count=1, Spawn Time=8.05, Use Spawn Probability=false)
  ]
  particleSpawnScript [
   &InitializeParticle(
    Color=<rgba 0.209 0.3802 0.7 5>
    Lifetime=&UniformRangedFloat(Minimum=0.45, Maximum=0.45)
    Mesh Scale=<1.5 1.5 1.5>
    Module.Write Color=true
    Module.Write Lifetime=true
    Module.Write Mass=false
    Module.Write Material Random=true
    Module.Write Position=false
    Module.Write Scale=false
    Module.Write SpriteRotation=true
    Module.Write SpriteSize=true
    Position=&SimulationPosition()
    Sprite Rotation=&UniformRangedFloat(Minimum=0, Maximum=1)
    Sprite Size=<1.7 1.9>
   )
   &SampleSkeletalMeshSkeleton(Mesh=DataInterface() {propertyValues {
      DefaultMesh=null
      ExcludeBoneName=root
      OptionalComponentClass=null
      PreviewMesh=&Watering_Can_Strawberry_Gadget
      SkinningMode=SkinOnTheFly
      SoftSourceActor=null
      SourceMode=Default
      UvSetIndex=0
      WholeMeshLOD=-1
      bExcludeBone=false
      bReadDeformedGeometry=true
      bRequireCurrentFrameData=true
      ComponentTags []
      FilteredBones [strawberry_02]
      FilteredSockets []
      MeshUserParameter {parameter {name=None, typeDefHandle {registeredTypeIndex=1}, varData []}}
      SamplingRegions []
     }})
   &SkeletalMeshSkeletonLocation(Module.Inherit Velocity=false, Module.Use Sampled Bone as Ribbon ID=true)
   &SphereLocation(MaskSpawnGroup=false, Module.Override Randomness=false, Sphere Radius=5, Surface Only=false)
   &SystemLocation(MaskSpawnGroup=false, Offset=<0 65 150>) {enabled=false}
   &AddVelocity(
    Coordinate Space=::Simulation
    Velocity=&Multiply_Vector(
     A=$Output.SphereLocation.SphereNormal
     B=&VectorFromFloat(Value=&UniformRangedFloat(Minimum=2, Maximum=5))
    )
   )
   &AddVelocity(Coordinate Space=::Simulation, Velocity=&UniformRangedVector(Minimum=<0 0 75>, Maximum=<1 1 100>)) {module=AddVelocity001}
  ]
  particleUpdateScript [
   &ParticleState()
   &Drag(Drag=3, Use Linear Drag=true, Use Rotational Drag=false) {enabled=false}
   &GravityForce(Gravity=&UniformRangedVector(Minimum=<0 0 -600>, Maximum=<1 1 -750>))
   &SolveForcesAndVelocity(Clamp Velocity=false, Limit Acceleration=false)
   &ScaleSpriteSize(Scale Factor=&Vector2DFromFloat(Value=&FloatFromCurve(FloatCurve=DataInterface() {propertyValues {Curve {keys [{arriveTangent=35.4138, interpMode=RCIM_Cubic, leaveTangent=35.4138, tangentMode=RCTM_User, time=0, value=10}, <at 1 44 RCIM_Cubic>]}}})))
   &CameraOffset(Camera Offset Amount=10)
   &DynamicMaterialParameters(
    Index 0 Param 1=&FloatFromCurve(FloatCurve=DataInterface() {propertyValues {Curve {keys [<at 0 1>,<at 1.0011 4>]}}})
    Index 0 Param 2=&FloatFromCurve(FloatCurve=DataInterface() {propertyValues {Curve {keys [<at -0.0046 0.09>,<at 1.0011 1.85>]}}})
    Index 0 Param 3=0.08
    Index 0 Param 4=2.83
    Module.Param0WriteEnabled=true
    Module.Param1WriteEnabled=true
    Module.Param2WriteEnabled=true
    Module.Param3WriteEnabled=true
   )
  ]
  rendererClasses [&NiagaraSpriteRendererProperties]
  renderers [{rendererClass=&NiagaraSpriteRendererProperties}]
 }
 Flare_Splash_Pour {
  emitterUpdateScript [
   &EmitterState(Life Cycle Mode=::NewEnumerator1, Loop Behavior=::NewEnumerator1, Loop Duration=10, UseLoopDelay=false)
   &SpawnBurst_Instantaneous(Spawn Count=1, Spawn Time=1.82, Use Spawn Probability=false)
  ]
  particleSpawnScript [
   &InitializeParticle(
    Color=<rgba 0.209 0.3802 0.7 5>
    Lifetime=&UniformRangedFloat(Minimum=0.45, Maximum=0.45)
    Mesh Scale=<1.5 1.5 1.5>
    Module.Write Color=true
    Module.Write Lifetime=true
    Module.Write Mass=false
    Module.Write Material Random=true
    Module.Write Position=false
    Module.Write Scale=false
    Module.Write SpriteRotation=true
    Module.Write SpriteSize=true
    Position=&SimulationPosition()
    Sprite Rotation=&UniformRangedFloat(Minimum=0, Maximum=1)
    Sprite Size=<2.55 2.85>
   )
   &SampleSkeletalMeshSkeleton(
    Mesh=DataInterface() {propertyValues {
      DefaultMesh=null
      ExcludeBoneName=root
      OptionalComponentClass=null
      PreviewMesh=&Watering_Can_Strawberry_Gadget
      SkinningMode=SkinOnTheFly
      SoftSourceActor=null
      SourceMode=Default
      UvSetIndex=0
      WholeMeshLOD=-1
      bExcludeBone=false
      bReadDeformedGeometry=true
      bRequireCurrentFrameData=true
      ComponentTags []
      FilteredBones []
      FilteredSockets [FX_Socket]
      MeshUserParameter {parameter {name=None, typeDefHandle {registeredTypeIndex=1}, varData []}}
      SamplingRegions []
     }}
    Spawn Mode:ENiagaraSkeletalSpawning=::NewEnumerator1
   )
   &SkeletalMeshSkeletonLocation(Module.Inherit Velocity=false, Module.Use Sampled Bone as Ribbon ID=true, Sampled Position Offset:Vector3f=<0 0 -5>)
   &SphereLocation(MaskSpawnGroup=false, Module.Override Randomness=false, Sphere Radius=5, Surface Only=false)
   &SystemLocation(MaskSpawnGroup=false, Offset=<0 65 150>) {enabled=false}
   &AddVelocity(
    Coordinate Space=::Simulation
    Velocity=&Multiply_Vector(
     A=$Output.SphereLocation.SphereNormal
     B=&VectorFromFloat(Value=&UniformRangedFloat(Minimum=2, Maximum=5))
    )
   )
   &AddVelocity(Coordinate Space=::Simulation, Velocity=&UniformRangedVector(Minimum=<0 0 75>, Maximum=<1 1 100>)) {module=AddVelocity001}
  ]
  particleUpdateScript [
   &ParticleState()
   &Drag(Drag=3, Use Linear Drag=true, Use Rotational Drag=false) {enabled=false}
   &GravityForce(Gravity=&UniformRangedVector(Minimum=<0 0 -600>, Maximum=<1 1 -750>))
   &SolveForcesAndVelocity(Clamp Velocity=false, Limit Acceleration=false)
   &ScaleSpriteSize(Scale Factor=&Vector2DFromFloat(Value=&FloatFromCurve(FloatCurve=DataInterface() {propertyValues {Curve {keys [{arriveTangent=35.4138, interpMode=RCIM_Cubic, leaveTangent=35.4138, tangentMode=RCTM_User, time=0, value=10}, <at 1 44 RCIM_Cubic>]}}})))
   &CameraOffset(Camera Offset Amount=10)
   &DynamicMaterialParameters(
    Index 0 Param 1=&FloatFromCurve(FloatCurve=DataInterface() {propertyValues {Curve {keys [<at 0 2>,<at 1.0011 4>]}}})
    Index 0 Param 2=&FloatFromCurve(FloatCurve=DataInterface() {propertyValues {Curve {keys [<at -0.0046 0.09>,<at 1.0011 1.85>]}}})
    Index 0 Param 3=0.08
    Index 0 Param 4=2.83
    Module.Param0WriteEnabled=true
    Module.Param1WriteEnabled=true
    Module.Param2WriteEnabled=true
    Module.Param3WriteEnabled=true
   )
  ]
  rendererClasses [&NiagaraSpriteRendererProperties]
  renderers [{rendererClass=&NiagaraSpriteRendererProperties}]
 }
 Flare_end {
  emitterUpdateScript [
   &EmitterState(Life Cycle Mode=::NewEnumerator1, Loop Behavior=::NewEnumerator1, Loop Duration=10, UseLoopDelay=false)
   &SpawnBurst_Instantaneous(Spawn Count=1, Spawn Time=7.95, Use Spawn Probability=false)
   &SpawnBurst_Instantaneous(Spawn Count=1, Use Spawn Probability=false) {enabled=false, module=SpawnBurst_Instantaneous001}
  ]
  particleSpawnScript [
   &InitializeParticle(
    Color=<rgba 50 15.4499 23.2308 1>
    Lifetime=&UniformRangedFloat(Minimum=0.1, Maximum=0.1)
    Mesh Scale=<1.5 1.5 1.5>
    Module.Write Color=true
    Module.Write Lifetime=true
    Module.Write Mass=true
    Module.Write Material Random=true
    Module.Write Position=false
    Module.Write Scale=false
    Module.Write SpriteRotation=true
    Module.Write SpriteSize=true
    Position=&SimulationPosition()
    Sprite Rotation=&UniformRangedFloat(Minimum=0, Maximum=1)
    Sprite Size=<20 20>
   )
   &SampleSkeletalMeshSkeleton(Mesh=DataInterface() {propertyValues {
      DefaultMesh=null
      ExcludeBoneName=root
      OptionalComponentClass=null
      PreviewMesh=&Watering_Can_Strawberry_Gadget
      SkinningMode=SkinOnTheFly
      SoftSourceActor=null
      SourceMode=Default
      UvSetIndex=0
      WholeMeshLOD=-1
      bExcludeBone=false
      bReadDeformedGeometry=true
      bRequireCurrentFrameData=true
      ComponentTags []
      FilteredBones [strawberry_02]
      FilteredSockets []
      MeshUserParameter {parameter {name=None, typeDefHandle {registeredTypeIndex=1}, varData []}}
      SamplingRegions []
     }})
   &SkeletalMeshSkeletonLocation(Module.Inherit Velocity=false, Module.Use Sampled Bone as Ribbon ID=true)
   &SphereLocation(MaskSpawnGroup=false, Module.Override Randomness=false, Offset=<5 -15 0>, Sphere Radius=0, Surface Only=false)
   &SystemLocation(MaskSpawnGroup=false, Offset=<0 65 150>) {enabled=false}
   &FN_Light_v2(Light Radius=100, Position Offset:Vector3f=<0 -30 0>)
   /SetVariables(Particles.Light Color:LinearColor=<rgba 5 1.545 2.3231 1>) {isSetParameters=true, module=SetVariables_#1}
  ]
  particleUpdateScript [
   &ParticleState()
   &Drag(Drag=3, Use Linear Drag=true, Use Rotational Drag=false)
   &SolveForcesAndVelocity(Clamp Velocity=false, Limit Acceleration=false)
   &ScaleSpriteSize(Scale Factor=&Vector2DFromFloat(Value=&FloatFromCurve(FloatCurve=DataInterface() {propertyValues {Curve {keys [{arriveTangent=1.2387, interpMode=RCIM_Cubic, leaveTangent=1.2387, tangentMode=RCTM_User, time=0, value=0.5}, <at 1 1 RCIM_Cubic>]}}})))
   &CameraOffset(Camera Offset Amount=10)
  ]
  rendererClasses [&NiagaraSpriteRendererProperties, &NiagaraLightRendererProperties]
  renderers [
   {rendererClass=&NiagaraSpriteRendererProperties}
   {rendererClass=&NiagaraLightRendererProperties, rendererIndex=1}
  ]
 }
 PuffSmoke_1 {
  emitterUpdateScript [
   &EmitterState(Life Cycle Mode=::NewEnumerator1, Loop Behavior=::NewEnumerator1, Loop Duration=10, UseLoopDelay=false)
   &SpawnBurst_Instantaneous(Spawn Count=3, Spawn Time=1.82, Use Spawn Probability=false)
  ]
  particleSpawnScript [
   &InitializeParticle(
    Lifetime=&UniformRangedFloat(Minimum=1, Maximum=1)
    Mesh Scale=<1.5 1.5 1.5>
    Module.Write Color=false
    Module.Write Lifetime=true
    Module.Write Mass=true
    Module.Write Material Random=true
    Module.Write Position=false
    Module.Write Scale=false
    Module.Write SpriteRotation=true
    Module.Write SpriteSize=true
    Position=&SimulationPosition()
    Sprite Rotation=&UniformRangedFloat(Minimum=0, Maximum=1)
    Sprite Size=&UniformRangedVector2D(
     Minimum=&Vector2DFromFloat(Value=130)
     Maximum=&Vector2DFromFloat(Value=110)
    )
   )
   &SampleSkeletalMeshSkeleton(
    Mesh=DataInterface() {propertyValues {
      DefaultMesh=null
      ExcludeBoneName=root
      OptionalComponentClass=null
      PreviewMesh=&Watering_Can_Strawberry_Gadget
      SkinningMode=SkinOnTheFly
      SoftSourceActor=null
      SourceMode=Default
      UvSetIndex=0
      WholeMeshLOD=-1
      bExcludeBone=false
      bReadDeformedGeometry=true
      bRequireCurrentFrameData=true
      ComponentTags []
      FilteredBones []
      FilteredSockets [FX_Socket]
      MeshUserParameter {parameter {name=None, typeDefHandle {registeredTypeIndex=1}, varData []}}
      SamplingRegions []
     }}
    Spawn Mode:ENiagaraSkeletalSpawning=::NewEnumerator1
   )
   &SkeletalMeshSkeletonLocation(Module.Inherit Velocity=false, Module.Use Sampled Bone as Ribbon ID=true)
   &SphereLocation(MaskSpawnGroup=false, Module.Override Randomness=false, Sphere Radius=0, Surface Only=false)
   &AddVelocity(Velocity=&UniformRangedVector(Minimum=<-20 -20 0>, Maximum=<20 20 5>))
  ]
  particleUpdateScript [
   &ParticleState()
   &Drag(Drag=3, Use Linear Drag=true, Use Rotational Drag=false)
   &SolveForcesAndVelocity(Clamp Velocity=false, Limit Acceleration=false)
   &Color(Color=&ColorFromCurve(ColorCurve=DataInterface() {propertyValues {
       AlphaCurve {keys [<at -0.0023 0 RCIM_Cubic>,<at 0.1821 0.15>,<at 0.9977 -0.003 RCIM_Cubic>]}
       BlueCurve {keys [<at -0.0023 0.2891 RCIM_Cubic>,<at 0.9977 0.2891 RCIM_Cubic>]}
       GreenCurve {keys [<at -0.0023 0.3324 RCIM_Cubic>,<at 0.9977 0.3324 RCIM_Cubic>]}
       RedCurve {keys [<at -0.0023 0.3854 RCIM_Cubic>,<at 0.9977 0.3854 RCIM_Cubic>]}
      }}))
   &SubUVAnimation(Number Of Frames=20)
   &ScaleSpriteSize(Scale Factor=&Vector2DFromFloat(Value=&FloatFromCurve(FloatCurve=DataInterface() {propertyValues {Curve {keys [<at 0 0.5 RCIM_Cubic>,<at 1 1 RCIM_Cubic>]}}})))
  ]
  rendererClasses [&NiagaraSpriteRendererProperties]
  renderers [{rendererClass=&NiagaraSpriteRendererProperties}]
 }
 PuffSmoke_2 {
  emitterUpdateScript [
   &EmitterState(Life Cycle Mode=::NewEnumerator1, Loop Behavior=::NewEnumerator1, Loop Duration=10, UseLoopDelay=false)
   &SpawnBurst_Instantaneous(Spawn Count=3, Spawn Time=2.7, Use Spawn Probability=false)
  ]
  particleSpawnScript [
   &InitializeParticle(
    Lifetime=&UniformRangedFloat(Minimum=1, Maximum=1)
    Mesh Scale=<1.5 1.5 1.5>
    Module.Write Color=false
    Module.Write Lifetime=true
    Module.Write Mass=true
    Module.Write Material Random=true
    Module.Write Position=false
    Module.Write Scale=false
    Module.Write SpriteRotation=true
    Module.Write SpriteSize=true
    Position=&SimulationPosition()
    Sprite Rotation=&UniformRangedFloat(Minimum=0, Maximum=1)
    Sprite Size=&UniformRangedVector2D(
     Minimum=&Vector2DFromFloat(Value=130)
     Maximum=&Vector2DFromFloat(Value=110)
    )
   )
   &SampleSkeletalMeshSkeleton(
    Mesh=DataInterface() {propertyValues {
      DefaultMesh=null
      ExcludeBoneName=root
      OptionalComponentClass=null
      PreviewMesh=&Watering_Can_Strawberry_Gadget
      SkinningMode=SkinOnTheFly
      SoftSourceActor=null
      SourceMode=Default
      UvSetIndex=0
      WholeMeshLOD=-1
      bExcludeBone=false
      bReadDeformedGeometry=true
      bRequireCurrentFrameData=true
      ComponentTags []
      FilteredBones []
      FilteredSockets [FX_Socket]
      MeshUserParameter {parameter {name=None, typeDefHandle {registeredTypeIndex=1}, varData []}}
      SamplingRegions []
     }}
    Spawn Mode:ENiagaraSkeletalSpawning=::NewEnumerator1
   )
   &SkeletalMeshSkeletonLocation(Module.Inherit Velocity=false, Module.Use Sampled Bone as Ribbon ID=true)
   &SphereLocation(MaskSpawnGroup=false, Module.Override Randomness=false, Sphere Radius=0, Surface Only=false)
   &AddVelocity(Velocity=&UniformRangedVector(Minimum=<-20 -20 0>, Maximum=<20 20 5>))
  ]
  particleUpdateScript [
   &ParticleState()
   &Drag(Drag=3, Use Linear Drag=true, Use Rotational Drag=false)
   &SolveForcesAndVelocity(Clamp Velocity=false, Limit Acceleration=false)
   &Color(Color=&ColorFromCurve(ColorCurve=DataInterface() {propertyValues {
       AlphaCurve {keys [<at -0.0023 0 RCIM_Cubic>,<at 0.1821 0.15>,<at 0.9977 -0.003 RCIM_Cubic>]}
       BlueCurve {keys [<at -0.0023 0.2891 RCIM_Cubic>,<at 0.9977 0.2891 RCIM_Cubic>]}
       GreenCurve {keys [<at -0.0023 0.3324 RCIM_Cubic>,<at 0.9977 0.3324 RCIM_Cubic>]}
       RedCurve {keys [<at -0.0023 0.3854 RCIM_Cubic>,<at 0.9977 0.3854 RCIM_Cubic>]}
      }}))
   &SubUVAnimation(Number Of Frames=20)
   &ScaleSpriteSize(Scale Factor=&Vector2DFromFloat(Value=&FloatFromCurve(FloatCurve=DataInterface() {propertyValues {Curve {keys [<at 0 0.5 RCIM_Cubic>,<at 1 1 RCIM_Cubic>]}}})))
  ]
  rendererClasses [&NiagaraSpriteRendererProperties]
  renderers [{rendererClass=&NiagaraSpriteRendererProperties}]
 }
 Sparkles1 {
  emitterUpdateScript [
   &EmitterState(Life Cycle Mode=::NewEnumerator1, Loop Behavior=::NewEnumerator1, Loop Duration=10, UseLoopDelay=true)
   &SpawnBurst_Instantaneous(Spawn Count=&UniformRangedInt(Minimum=20, Maximum=40), Spawn Time=2.7, Use Spawn Probability=false)
   &SpawnRate(SpawnRate=50, Use Spawn Probability=false) {enabled=false}
  ]
  particleSpawnScript [
   &InitializeParticle(
    Color=<rgba 0.25 0.25 1 1>
    Lifetime=&UniformRangedFloat(Minimum=0.1, Maximum=0.65)
    Mesh Scale=<1.5 1.5 1.5>
    Module.Write Color=true
    Module.Write Lifetime=true
    Module.Write Mass=false
    Module.Write Material Random=true
    Module.Write Position=false
    Module.Write Scale=false
    Module.Write SpriteRotation=true
    Module.Write SpriteSize=true
    Position=&SimulationPosition()
    Sprite Rotation=&UniformRangedFloat(Minimum=0, Maximum=1)
    Sprite Size=&UniformRangedVector2D(
     Minimum=&Vector2DFromFloat(Value=8)
     Maximum=&Vector2DFromFloat(Value=15)
    )
   )
   &SampleSkeletalMeshSkeleton(
    Mesh=DataInterface() {propertyValues {
      DefaultMesh=null
      ExcludeBoneName=root
      OptionalComponentClass=null
      PreviewMesh=&Watering_Can_Strawberry_Gadget
      SkinningMode=SkinOnTheFly
      SoftSourceActor=null
      SourceMode=Default
      UvSetIndex=0
      WholeMeshLOD=-1
      bExcludeBone=false
      bReadDeformedGeometry=true
      bRequireCurrentFrameData=true
      ComponentTags []
      FilteredBones []
      FilteredSockets [FX_Socket]
      MeshUserParameter {parameter {name=None, typeDefHandle {registeredTypeIndex=1}, varData []}}
      SamplingRegions []
     }}
    Spawn Mode:ENiagaraSkeletalSpawning=::NewEnumerator1
   )
   &SkeletalMeshSkeletonLocation(Module.Inherit Velocity=false, Module.Use Sampled Bone as Ribbon ID=true)
   &SphereLocation(MaskSpawnGroup=false, Module.Override Randomness=false, Sphere Radius=0, Surface Only=false)
   &SubUVAnimation(Number Of Frames=1, SubUV Animation Mode=::NewEnumerator1) {enabled=false}
   &SphereLocation(MaskSpawnGroup=false, Module.Override Randomness=false, Sphere Radius=20, Surface Only=false) {module=SphereLocation001}
   &AddVelocity(Velocity=&Multiply_Vector(A=$Output.SphereLocation001.SphereNormal, B=&VectorFromFloat(Value=100)))
  ]
  particleUpdateScript [
   &ParticleState()
   &Color(Color=&ColorFromCurve(ColorCurve=DataInterface() {propertyValues {
       AlphaCurve {keys [
         <at -0.0023 0 RCIM_Cubic>
         {arriveTangent=0.2428, interpMode=RCIM_Cubic, leaveTangent=0.2428, time=0.1, value=1}
         {arriveTangent=-0.0163, interpMode=RCIM_Cubic, leaveTangent=-0.0163, time=0.2779, value=0.0681}
         {arriveTangent=-0.0735, interpMode=RCIM_Cubic, leaveTangent=-0.0735, time=0.4489, value=0.9943}
         {arriveTangent=0.0148, interpMode=RCIM_Cubic, leaveTangent=0.0148, time=0.6472, value=0.0409}
         {arriveTangent=-0.1252, interpMode=RCIM_Cubic, leaveTangent=-0.1252, time=0.8337, value=1}
         <at 0.9977 -0.003 RCIM_Cubic>
        ]}
       BlueCurve {keys [<at -0.0023 10 RCIM_Cubic>,<at 0.448 7.6655>,<at 0.9977 2.2 RCIM_Cubic>]}
       GreenCurve {keys [<at -0.0023 1.3875 RCIM_Cubic>,<at 0.448 5.0649>,<at 0.9977 10 RCIM_Cubic>]}
       RedCurve {keys [<at -0.0023 0.25 RCIM_Cubic>,<at 0.448 20>,<at 0.9977 0.25 RCIM_Cubic>]}
      }}))
   &GravityForce(Gravity=<0 0 980>) {enabled=false}
   &AccelerationForce(Acceleration=<0 0 200>) {enabled=false}
   /Niagara/Modules/Update/Forces/CurlNoiseForce(Noise Strength=100, Pan Noise=false)
   &Drag(Drag=2, Use Linear Drag=true, Use Rotational Drag=false)
   &SolveForcesAndVelocity(Clamp Velocity=false, Limit Acceleration=false)
  ]
  rendererClasses [&NiagaraSpriteRendererProperties]
  renderers [{rendererClass=&NiagaraSpriteRendererProperties}]
 }
 Sparkles_End {
  emitterUpdateScript [
   &EmitterState(Life Cycle Mode=::NewEnumerator1, Loop Behavior=::NewEnumerator1, Loop Duration=10, UseLoopDelay=false)
   &SpawnBurst_Instantaneous(Spawn Count=&UniformRangedInt(Minimum=15, Maximum=20), Spawn Time=8.05, Use Spawn Probability=false)
   &SpawnRate(SpawnRate=50, Use Spawn Probability=false) {enabled=false}
   &SpawnBurst_Instantaneous(Spawn Count=15, Use Spawn Probability=false) {enabled=false, module=SpawnBurst_Instantaneous001}
  ]
  particleSpawnScript [
   &InitializeParticle(
    Color=<rgba 0.25 0.25 1 1>
    Lifetime=&UniformRangedFloat(Minimum=0.3, Maximum=0.85)
    Mesh Scale=<1.5 1.5 1.5>
    Module.Write Color=true
    Module.Write Lifetime=true
    Module.Write Mass=false
    Module.Write Material Random=true
    Module.Write Position=false
    Module.Write Scale=false
    Module.Write SpriteRotation=true
    Module.Write SpriteSize=true
    Position=&SimulationPosition()
    Sprite Rotation=&UniformRangedFloat(Minimum=0, Maximum=1)
    Sprite Size=&UniformRangedVector2D(
     Minimum=&Vector2DFromFloat(Value=8)
     Maximum=&Vector2DFromFloat(Value=15)
    )
   )
   &SubUVAnimation(Number Of Frames=1, SubUV Animation Mode=::NewEnumerator1) {enabled=false}
   &SampleSkeletalMeshSkeleton(Mesh=DataInterface() {propertyValues {
      DefaultMesh=null
      ExcludeBoneName=root
      OptionalComponentClass=null
      PreviewMesh=&Watering_Can_Strawberry_Gadget
      SkinningMode=SkinOnTheFly
      SoftSourceActor=null
      SourceMode=Default
      UvSetIndex=0
      WholeMeshLOD=-1
      bExcludeBone=false
      bReadDeformedGeometry=true
      bRequireCurrentFrameData=true
      ComponentTags []
      FilteredBones [strawberry_02]
      FilteredSockets []
      MeshUserParameter {parameter {name=None, typeDefHandle {registeredTypeIndex=1}, varData []}}
      SamplingRegions []
     }})
   &SkeletalMeshSkeletonLocation(Module.Inherit Velocity=false, Module.Use Sampled Bone as Ribbon ID=true)
   &SphereLocation(MaskSpawnGroup=false, Module.Override Randomness=false, Sphere Radius=30, Surface Only=true, Surface Only Band Thickness=5)
   &AddVelocity(
    Coordinate Space=::Simulation
    Velocity=&Multiply_Vector(
     A=$Output.SphereLocation.SphereNormal
     B=&VectorFromFloat(Value=&UniformRangedFloat(Minimum=200, Maximum=300))
    )
   )
   &SystemLocation(MaskSpawnGroup=false, Offset=<0 65 150>) {enabled=false}
  ]
  particleUpdateScript [
   &ParticleState()
   &Color(Color=&ColorFromCurve(ColorCurve=DataInterface() {propertyValues {
       AlphaCurve {keys [
         {arriveTangent=0.2428, interpMode=RCIM_Cubic, time=0.1, value=1}
         {arriveTangent=-0.0163, interpMode=RCIM_Cubic, leaveTangent=-0.0163, time=0.2779, value=0.0681}
         {arriveTangent=-0.0735, interpMode=RCIM_Cubic, leaveTangent=-0.0735, time=0.4489, value=0.9943}
         {arriveTangent=0.0148, interpMode=RCIM_Cubic, leaveTangent=0.0148, time=0.6472, value=0.0409}
         {arriveTangent=-0.1252, interpMode=RCIM_Cubic, leaveTangent=-0.1252, time=0.8337, value=1}
         <at 0.9977 -0.003 RCIM_Cubic>
        ]}
       BlueCurve {keys [<at -0.0023 10 RCIM_Cubic>,<at 0.448 0.8483>,<at 0.9977 2.2 RCIM_Cubic>]}
       GreenCurve {keys [<at -0.0023 1.3875 RCIM_Cubic>,<at 0.448 0>,<at 0.9977 10 RCIM_Cubic>]}
       RedCurve {keys [<at -0.0023 0.25 RCIM_Cubic>,<at 0.448 20>,<at 0.9977 0.25 RCIM_Cubic>]}
      }}))
   &GravityForce(Gravity=<0 0 400>) {enabled=false}
   &AccelerationForce(Acceleration=<0 0 200>) {enabled=false}
   &CurlNoiseForce(Noise Strength=100, Pan Noise=false)
   &Drag(Drag=8, Use Linear Drag=true, Use Rotational Drag=false)
   &SolveForcesAndVelocity(Clamp Velocity=false, Limit Acceleration=false)
  ]
  rendererClasses [&NiagaraSpriteRendererProperties]
  renderers [{rendererClass=&NiagaraSpriteRendererProperties}]
 }
 Sparkles_UpwardMotion {
  emitterUpdateScript [
   &EmitterState(Life Cycle Mode=::NewEnumerator1, Loop Behavior=::NewEnumerator1, Loop Delay=2.8, Loop Duration=0.3, UseLoopDelay=true)
   &SpawnBurst_Instantaneous(Spawn Count=&UniformRangedInt(Minimum=20, Maximum=40), Spawn Time=2.8, Use Spawn Probability=false) {enabled=false}
   &SpawnRate(SpawnRate=50, Use Spawn Probability=false)
  ]
  particleSpawnScript [
   &InitializeParticle(
    Color=<rgba 0.25 0.25 1 1>
    Lifetime=&UniformRangedFloat(Minimum=0.1, Maximum=0.6)
    Mesh Scale=<1.5 1.5 1.5>
    Module.Write Color=true
    Module.Write Lifetime=true
    Module.Write Mass=false
    Module.Write Material Random=true
    Module.Write Position=false
    Module.Write Scale=false
    Module.Write SpriteRotation=true
    Module.Write SpriteSize=true
    Position=&SimulationPosition()
    Sprite Rotation=&UniformRangedFloat(Minimum=0, Maximum=1)
    Sprite Size=&UniformRangedVector2D(
     Minimum=&Vector2DFromFloat(Value=8)
     Maximum=&Vector2DFromFloat(Value=15)
    )
   )
   &SampleSkeletalMeshSkeleton(Mesh=DataInterface() {propertyValues {
      DefaultMesh=null
      ExcludeBoneName=root
      OptionalComponentClass=null
      PreviewMesh=&Watering_Can_Strawberry_Gadget
      SkinningMode=SkinOnTheFly
      SoftSourceActor=null
      SourceMode=Default
      UvSetIndex=0
      WholeMeshLOD=-1
      bExcludeBone=false
      bReadDeformedGeometry=true
      bRequireCurrentFrameData=true
      ComponentTags []
      FilteredBones [strawberry_02]
      FilteredSockets []
      MeshUserParameter {parameter {name=None, typeDefHandle {registeredTypeIndex=1}, varData []}}
      SamplingRegions []
     }})
   &SkeletalMeshSkeletonLocation(Module.Inherit Velocity=false, Module.Use Sampled Bone as Ribbon ID=true)
   &SphereLocation(MaskSpawnGroup=false, Module.Override Randomness=false, Sphere Radius=0, Surface Only=false)
   &SubUVAnimation(Number Of Frames=1, SubUV Animation Mode=::NewEnumerator1) {enabled=false}
   &SphereLocation(MaskSpawnGroup=false, Module.Override Randomness=false, Sphere Radius=20, Surface Only=false) {module=SphereLocation001}
   &AddVelocity(Velocity=&Multiply_Vector(A=$Output.SphereLocation001.SphereNormal, B=&VectorFromFloat(Value=100))) {enabled=false}
   &InheritVelocity(Module.Limit Speed=true, Velocity Scale=<0.1 0.1 0.1>)
  ]
  particleUpdateScript [
   &ParticleState()
   &Color(Color=&ColorFromCurve(ColorCurve=DataInterface() {propertyValues {
       AlphaCurve {keys [
         <at -0.0023 0 RCIM_Cubic>
         {arriveTangent=0.2428, interpMode=RCIM_Cubic, leaveTangent=0.2428, time=0.1, value=1}
         {arriveTangent=-0.0163, interpMode=RCIM_Cubic, leaveTangent=-0.0163, time=0.2779, value=0.0681}
         {arriveTangent=-0.0735, interpMode=RCIM_Cubic, leaveTangent=-0.0735, time=0.4489, value=0.9943}
         {arriveTangent=0.0148, interpMode=RCIM_Cubic, leaveTangent=0.0148, time=0.6472, value=0.0409}
         {arriveTangent=-0.1252, interpMode=RCIM_Cubic, leaveTangent=-0.1252, time=0.8337, value=1}
         <at 0.9977 -0.003 RCIM_Cubic>
        ]}
       BlueCurve {keys [<at -0.0023 5 RCIM_Cubic>,<at 0.448 3.8328>,<at 0.9977 1.1 RCIM_Cubic>]}
       GreenCurve {keys [<at -0.0023 0.6937 RCIM_Cubic>,<at 0.448 2.5324>,<at 0.9977 5 RCIM_Cubic>]}
       RedCurve {keys [<at -0.0023 0.125 RCIM_Cubic>,<at 0.448 10>,<at 0.9977 0.125 RCIM_Cubic>]}
      }}))
   &CurlNoiseForce(Noise Strength=500, Pan Noise=false) {enabled=false}
   &GravityForce(Gravity=<0 0 980>) {enabled=false}
   &AccelerationForce(Acceleration=<0 0 200>) {enabled=false}
   &Drag(Drag=2, Use Linear Drag=true, Use Rotational Drag=false)
   &SolveForcesAndVelocity(Clamp Velocity=false, Limit Acceleration=false)
  ]
  rendererClasses [&NiagaraSpriteRendererProperties]
  renderers [{rendererClass=&NiagaraSpriteRendererProperties}]
 }
 Water_Bounce {
  emitterUpdateScript [
   &EmitterState(Life Cycle Mode=::NewEnumerator1, Loop Behavior=::NewEnumerator1, Loop Delay=6.4, Loop Duration=10, UseLoopDelay=false)
   &SpawnRate(SpawnRate=200, Use Spawn Probability=false) {enabled=false}
   &SpawnBurst_Instantaneous(Spawn Count=&UniformRangedInt(Minimum=5, Maximum=10), Spawn Time=4.19, Use Spawn Probability=false)
   &SpawnBurst_Instantaneous(Spawn Count=&UniformRangedInt(Minimum=5, Maximum=10), Spawn Time=4.67, Use Spawn Probability=false) {module=SpawnBurst_Instantaneous001}
   &SpawnBurst_Instantaneous(Spawn Count=&UniformRangedInt(Minimum=5, Maximum=10), Spawn Time=5.06, Use Spawn Probability=false) {module=SpawnBurst_Instantaneous002}
   &SpawnBurst_Instantaneous(Spawn Count=&UniformRangedInt(Minimum=5, Maximum=10), Spawn Time=5.5, Use Spawn Probability=false) {module=SpawnBurst_Instantaneous003}
   &SpawnBurst_Instantaneous(Spawn Count=&UniformRangedInt(Minimum=5, Maximum=10), Spawn Time=5.9, Use Spawn Probability=false) {module=SpawnBurst_Instantaneous004}
   &SpawnBurst_Instantaneous(Spawn Count=&UniformRangedInt(Minimum=2, Maximum=5), Spawn Time=6.36, Use Spawn Probability=false) {enabled=false, module=SpawnBurst_Instantaneous005}
   &SpawnBurst_Instantaneous(Spawn Count=10, Spawn Time=8.06, Use Spawn Probability=false) {enabled=false, module=SpawnBurst_Instantaneous006}
  ]
  particleSpawnScript [
   &InitializeParticle(
    Color=<rgba 1 1 1 0.3>
    Lifetime=&UniformRangedFloat(Minimum=0.2, Maximum=0.4)
    Mass=&UniformRangedFloat(Minimum=0.4, Maximum=2)
    Module.Write Color=true
    Module.Write Lifetime=true
    Module.Write Mass=false
    Module.Write Material Random=true
    Module.Write Position=false
    Module.Write Scale=false
    Module.Write SpriteRotation=true
    Module.Write SpriteSize=true
    Position=&SimulationPosition()
    Sprite Rotation=&NormalizedAngleToDegrees(NormalizedAngle=&UniformRangedFloat(Minimum=0, Maximum=1))
    Sprite Size=&UniformRangedVector2D(Minimum=<1.5 2>, Maximum=<2 3>)
   )
   &SampleSkeletalMeshSkeleton(
    Mesh=DataInterface() {propertyValues {
      DefaultMesh=null
      ExcludeBoneName=root
      OptionalComponentClass=null
      PreviewMesh=&Watering_Can_Strawberry_Gadget
      SkinningMode=SkinOnTheFly
      SoftSourceActor=null
      SourceMode=Default
      UvSetIndex=0
      WholeMeshLOD=-1
      bExcludeBone=false
      bReadDeformedGeometry=true
      bRequireCurrentFrameData=true
      ComponentTags []
      FilteredBones []
      FilteredSockets [FX_Socket]
      MeshUserParameter {parameter {name=None, typeDefHandle {registeredTypeIndex=1}, varData []}}
      SamplingRegions []
     }}
    Spawn Mode:ENiagaraSkeletalSpawning=::NewEnumerator1
   )
   &SkeletalMeshSkeletonLocation(Inherit Velocity From Mesh=0.1, Module.Inherit Velocity=false, Module.Use Sampled Bone as Ribbon ID=true)
   &SphereLocation(
    Hemisphere Z=true
    MaskSpawnGroup=false
    Module.Override Randomness=false
    Non Uniform Scale=<1 1 0>
    Sphere Radius=20
    Surface Expansion Mode:ENiagaraExpansionMode=::NewEnumerator2
    Surface Only=true
    Surface Only Band Thickness=0.1
   )
   &CylinderLocation(
    Cylinder Height=1
    Cylinder Radius=20
    Lathe Profile=DataInterface() {propertyValues {Curve {keys []}}}
    MaskSpawnGroup=false
    Module.Enable Lathe Profile=false
    Module.Override Randomness=false
    Module.Surface Only=true
    Surface Only Band Thickness=0
    Use Endcaps In Surface Only Mode=false
   ) {enabled=false}
   &AddVelocity(
    Coordinate Space=::Simulation
    Velocity=&Multiply_Vector(
     A=$Output.SphereLocation.SphereNormal
     B=&VectorFromFloat(Value=&UniformRangedFloat(Minimum=60, Maximum=120))
    )
   )
   &AddVelocity(Coordinate Space=::Simulation, Velocity=&UniformRangedVector(Minimum=<0 0 50>, Maximum=<0 0 100>)) {module=AddVelocity001}
   &ApplyInitialForces()
   &CameraOffset(Camera Offset Amount=5)
  ]
  particleUpdateScript [
   &ParticleState()
   &GravityForce(Gravity=&UniformRangedVector(Minimum=<0 0 -50>, Maximum=<1 1 -980>))
   &SolveForcesAndVelocity(Clamp Velocity=false, Limit Acceleration=false)
   &ScaleSpriteSizeBySpeed(Max Scale Factor=<1 4>)
  ]
  rendererClasses [&NiagaraSpriteRendererProperties]
  renderers [{rendererClass=&NiagaraSpriteRendererProperties}]
 }
 Water_Burst {
  emitterUpdateScript [
   &EmitterState(Life Cycle Mode=::NewEnumerator1, Loop Behavior=::NewEnumerator1, Loop Delay=6.4, Loop Duration=0.25, UseLoopDelay=true)
   &SpawnRate(SpawnRate=150, Use Spawn Probability=false)
   &SpawnBurst_Instantaneous(Spawn Count=50, Use Spawn Probability=false) {enabled=false}
  ]
  particleSpawnScript [
   &InitializeParticle(
    Color=<rgba 1 1 1 0.3>
    Lifetime=&UniformRangedFloat(Minimum=0.2, Maximum=0.6)
    Mass=&UniformRangedFloat(Minimum=0.4, Maximum=2)
    Module.Write Color=true
    Module.Write Lifetime=true
    Module.Write Mass=false
    Module.Write Material Random=true
    Module.Write Position=false
    Module.Write Scale=false
    Module.Write SpriteRotation=true
    Module.Write SpriteSize=true
    Position=&SimulationPosition()
    Sprite Rotation=&NormalizedAngleToDegrees(NormalizedAngle=&UniformRangedFloat(Minimum=0, Maximum=1))
    Sprite Size=&UniformRangedVector2D(Minimum=<2 3>, Maximum=<2.5 4>)
   )
   &SphereLocation(
    MaskSpawnGroup=false
    Module.Override Randomness=false
    Non Uniform Scale=<1 1 0.1>
    Offset=<0 0 40>
    Sphere Radius=20
    Surface Only=false
    Surface Only Band Thickness=50
   ) {enabled=false}
   &AddVelocity(Velocity=&Multiply_Vector(A=$Output.SphereLocation.SphereNormal, B=&VectorFromFloat(Value=50))) {enabled=false}
   &InheritVelocity(Module.Limit Speed=true, Velocity Limit=10000, Velocity Scale=<1 1 0>) {enabled=false}
   &SampleSkeletalMeshSkeleton(
    Mesh=DataInterface() {propertyValues {
      DefaultMesh=null
      ExcludeBoneName=root
      OptionalComponentClass=null
      PreviewMesh=&Watering_Can_Strawberry_Gadget
      SkinningMode=SkinOnTheFly
      SoftSourceActor=null
      SourceMode=Default
      UvSetIndex=0
      WholeMeshLOD=-1
      bExcludeBone=false
      bReadDeformedGeometry=true
      bRequireCurrentFrameData=true
      ComponentTags []
      FilteredBones []
      FilteredSockets [FX_Socket_2]
      MeshUserParameter {parameter {name=None, typeDefHandle {registeredTypeIndex=1}, varData []}}
      SamplingRegions []
     }}
    Spawn Mode:ENiagaraSkeletalSpawning=::NewEnumerator1
   )
   &SkeletalMeshSkeletonLocation(Inherit Velocity From Mesh=0.1, Module.Inherit Velocity=false, Module.Use Sampled Bone as Ribbon ID=true)
   &AddVelocityInCone(
    Cone Axis=&VectorFromCurve(
     VectorCurve=DataInterface() {propertyValues {
       XCurve {keys [
         <at 0 1 RCIM_Cubic>
         {arriveTangent=-4, interpMode=RCIM_Cubic, leaveTangent=-4, time=0.25, value=0}
         <at 0.5 -1 RCIM_Cubic>
         {arriveTangent=4, interpMode=RCIM_Cubic, leaveTangent=4, time=0.75, value=0}
         <at 1 1 RCIM_Cubic>
        ]}
       YCurve {keys [
         <at 0 0 RCIM_Cubic>
         <at 0.25 1 RCIM_Cubic>
         {arriveTangent=-4, interpMode=RCIM_Cubic, leaveTangent=-4, time=0.5, value=0}
         <at 0.75 -1 RCIM_Cubic>
         <at 1 0 RCIM_Cubic>
        ]}
       ZCurve {keys [<at -0.0011 0>,<at 0.9989 0>]}
      }}
     CurveIndex=$Emitter.NormalizedLoopAge
    )
    Cone Axis Coordinate Space=::Simulation
    Module.Use Velocity Falloff On Cone Axis=false
    Velocity Strength=&UniformRangedFloat(Minimum=100, Maximum=200)
   )
   &ApplyInitialForces()
   &InheritVelocity(Module.Limit Speed=true) {enabled=false, module=InheritVelocity001}
   /Game/Effects/Niagara/Modules/Position/FN_OffsetPositionByOrientationAxis(Offset Axis:Vector3f=<0 0 0>) {module=OffsetPositionByMeshOrientation}
  ]
  particleUpdateScript [
   &ParticleState()
   &GravityForce(Gravity=&UniformRangedVector(Minimum=<0 0 -100>, Maximum=<1 1 -400>))
   &SolveForcesAndVelocity(Clamp Velocity=false, Limit Acceleration=false)
   &ScaleSpriteSizeBySpeed(Max Scale Factor=<1 4>)
   &ScaleSpriteSize(Scale Factor=&Vector2DFromCurve(Vector2Curve=DataInterface() {propertyValues {XCurve {keys [<at 0 0.5>,<at 1 1>]}, YCurve {keys [<at 0 0.5>,<at 1 1>]}}}, ScaleCurve=<1 1>))
  ]
  rendererClasses [&NiagaraSpriteRendererProperties]
  renderers [{rendererClass=&NiagaraSpriteRendererProperties}]
 }
 Water_Burst001 {
  emitterUpdateScript [
   &EmitterState(Life Cycle Mode=::NewEnumerator1, Loop Behavior=::NewEnumerator1, Loop Delay=6.4, Loop Duration=0.25, UseLoopDelay=true)
   &SpawnRate(SpawnRate=75, Use Spawn Probability=false)
   &SpawnBurst_Instantaneous(Spawn Count=50, Use Spawn Probability=false) {enabled=false}
  ]
  particleSpawnScript [
   &InitializeParticle(
    Color=<rgba 1 1 1 0.3>
    Lifetime=&UniformRangedFloat(Minimum=0.2, Maximum=0.6)
    Mass=&UniformRangedFloat(Minimum=0.4, Maximum=2)
    Module.Write Color=true
    Module.Write Lifetime=true
    Module.Write Mass=false
    Module.Write Material Random=true
    Module.Write Position=false
    Module.Write Scale=false
    Module.Write SpriteRotation=true
    Module.Write SpriteSize=true
    Position=&SimulationPosition()
    Sprite Rotation=&NormalizedAngleToDegrees(NormalizedAngle=&UniformRangedFloat(Minimum=0, Maximum=1))
    Sprite Size=&UniformRangedVector2D(Minimum=<2 3>, Maximum=<3 4>)
   )
   &SampleSkeletalMeshSkeleton(
    Mesh=DataInterface() {propertyValues {
      DefaultMesh=null
      ExcludeBoneName=root
      OptionalComponentClass=null
      PreviewMesh=&Watering_Can_Strawberry_Gadget
      SkinningMode=SkinOnTheFly
      SoftSourceActor=null
      SourceMode=Default
      UvSetIndex=0
      WholeMeshLOD=-1
      bExcludeBone=false
      bReadDeformedGeometry=true
      bRequireCurrentFrameData=true
      ComponentTags []
      FilteredBones []
      FilteredSockets [FX_Socket_2]
      MeshUserParameter {parameter {name=None, typeDefHandle {registeredTypeIndex=1}, varData []}}
      SamplingRegions []
     }}
    Spawn Mode:ENiagaraSkeletalSpawning=::NewEnumerator1
   )
   &SkeletalMeshSkeletonLocation(Inherit Velocity From Mesh=0.1, Module.Inherit Velocity=false, Module.Use Sampled Bone as Ribbon ID=true)
   &SphereLocation(
    Hemisphere Z=true
    MaskSpawnGroup=false
    Module.Override Randomness=false
    Non Uniform Scale=<1 1 0.1>
    Sphere Radius=20
    Surface Only=false
    Surface Only Band Thickness=50
   )
   &AddVelocity(
    Coordinate Space=::Simulation
    Velocity=&Multiply_Vector(
     A=$Output.SphereLocation.SphereNormal
     B=&VectorFromFloat(Value=&UniformRangedFloat(Minimum=50, Maximum=200))
    )
   )
   &AddVelocity(Coordinate Space=::Simulation, Velocity=&UniformRangedVector(Minimum=<0 0 0>, Maximum=<1 1 100>)) {module=AddVelocity001}
   &InheritVelocity(Module.Limit Speed=true, Velocity Limit=10000, Velocity Scale=<1 1 0>) {enabled=false}
   &AddVelocityInCone(
    Cone Axis=&VectorFromCurve(
     VectorCurve=DataInterface() {propertyValues {
       XCurve {keys [
         <at 0 1 RCIM_Cubic>
         {arriveTangent=-4, interpMode=RCIM_Cubic, leaveTangent=-4, time=0.25, value=0}
         <at 0.5 -1 RCIM_Cubic>
         {arriveTangent=4, interpMode=RCIM_Cubic, leaveTangent=4, time=0.75, value=0}
         <at 1 1 RCIM_Cubic>
        ]}
       YCurve {keys [
         <at 0 0 RCIM_Cubic>
         <at 0.25 1 RCIM_Cubic>
         {arriveTangent=-4, interpMode=RCIM_Cubic, leaveTangent=-4, time=0.5, value=0}
         <at 0.75 -1 RCIM_Cubic>
         <at 1 0 RCIM_Cubic>
        ]}
       ZCurve {keys [<at -0.0011 0>,<at 0.9989 0>]}
      }}
     CurveIndex=$Emitter.NormalizedLoopAge
    )
    Cone Axis Coordinate Space=::Simulation
    Module.Use Velocity Falloff On Cone Axis=false
    Velocity Strength=200
   ) {enabled=false}
   &ApplyInitialForces()
   &InheritVelocity(Module.Limit Speed=true) {enabled=false, module=InheritVelocity001}
   &FN_OffsetPositionByOrientationAxis(Offset Axis:Vector3f=<0 0 0>) {enabled=false, module=OffsetPositionByMeshOrientation}
  ]
  particleUpdateScript [
   &ParticleState()
   &GravityForce(Gravity=&UniformRangedVector(Minimum=<0 0 -100>, Maximum=<1 1 -200>))
   &SolveForcesAndVelocity(Clamp Velocity=false, Limit Acceleration=false)
   &ScaleSpriteSizeBySpeed(Max Scale Factor=<1 4>)
   &ScaleSpriteSize(Scale Factor=&Vector2DFromCurve(Vector2Curve=DataInterface() {propertyValues {XCurve {keys [<at 0 0.5>,<at 1 1>]}, YCurve {keys [<at 0 0.5>,<at 1 1>]}}}, ScaleCurve=<1 1>))
  ]
  rendererClasses [&NiagaraSpriteRendererProperties]
  renderers [{rendererClass=&NiagaraSpriteRendererProperties}]
 }
 Water_End {
  emitterUpdateScript [
   &EmitterState(Life Cycle Mode=::NewEnumerator1, Loop Behavior=::NewEnumerator1, Loop Delay=6.4, Loop Duration=10, UseLoopDelay=false)
   &SpawnRate(SpawnRate=200, Use Spawn Probability=false) {enabled=false}
   &SpawnBurst_Instantaneous(Spawn Count=10, Spawn Time=7.95, Use Spawn Probability=false)
   &SpawnBurst_Instantaneous(Spawn Count=5, Spawn Time=8.05, Use Spawn Probability=false) {module=SpawnBurst_Instantaneous006}
  ]
  particleSpawnScript [
   &InitializeParticle(
    Color=<rgba 1 1 1 0.3>
    Lifetime=&UniformRangedFloat(Minimum=0.4, Maximum=0.6)
    Mass=&UniformRangedFloat(Minimum=0.4, Maximum=2)
    Module.Write Color=true
    Module.Write Lifetime=true
    Module.Write Mass=false
    Module.Write Material Random=true
    Module.Write Position=false
    Module.Write Scale=false
    Module.Write SpriteRotation=true
    Module.Write SpriteSize=true
    Position=&SimulationPosition()
    Sprite Rotation=&NormalizedAngleToDegrees(NormalizedAngle=&UniformRangedFloat(Minimum=0, Maximum=1))
    Sprite Size=&Vector2DFromFloat(Value=&MakeFloatFromVector2d(Vector 2D=&UniformRangedVector2D(Minimum=<1 0>, Maximum=<2 1>)))
   )
   &SampleSkeletalMeshSkeleton(Mesh=DataInterface() {propertyValues {
      DefaultMesh=null
      ExcludeBoneName=root
      OptionalComponentClass=null
      PreviewMesh=&Watering_Can_Strawberry_Gadget
      SkinningMode=SkinOnTheFly
      SoftSourceActor=null
      SourceMode=Default
      UvSetIndex=0
      WholeMeshLOD=-1
      bExcludeBone=false
      bReadDeformedGeometry=true
      bRequireCurrentFrameData=true
      ComponentTags []
      FilteredBones [strawberry_02]
      FilteredSockets [FX_Socket]
      MeshUserParameter {parameter {name=None, typeDefHandle {registeredTypeIndex=1}, varData []}}
      SamplingRegions []
     }})
   &SkeletalMeshSkeletonLocation(Inherit Velocity From Mesh=0.1, Module.Inherit Velocity=false, Module.Use Sampled Bone as Ribbon ID=true)
   &SphereLocation(MaskSpawnGroup=false, Module.Override Randomness=false, Sphere Radius=20, Surface Only=false, Surface Only Band Thickness=50)
   &AddVelocity(
    Coordinate Space=::Simulation
    Velocity=&Multiply_Vector(
     A=$Output.SphereLocation.SphereNormal
     B=&VectorFromFloat(Value=&UniformRangedFloat(Minimum=50, Maximum=100))
    )
   )
   &AddVelocity(Coordinate Space=::Simulation, Velocity=&UniformRangedVector(Minimum=<0 0 50>, Maximum=<0 0 200>)) {module=AddVelocity001}
   &ApplyInitialForces()
   &InheritVelocity(Module.Limit Speed=true, Velocity Scale=<0.1 0.1 0.1>) {enabled=false, module=InheritVelocity001}
  ]
  particleUpdateScript [
   &ParticleState()
   &GravityForce(Gravity=&UniformRangedVector(Minimum=<0 0 -100>, Maximum=<1 1 -980>))
   &SolveForcesAndVelocity(Clamp Velocity=false, Limit Acceleration=false)
   &ScaleSpriteSizeBySpeed(Max Scale Factor=<1 4>)
  ]
  rendererClasses [&NiagaraSpriteRendererProperties]
  renderers [{rendererClass=&NiagaraSpriteRendererProperties}]
 }
 Water_GroundSplashes {
  emitterUpdateScript [
   &EmitterState(Life Cycle Mode=::NewEnumerator1, Loop Behavior=::NewEnumerator1, Loop Delay=1.5, Loop Duration=1.55, UseLoopDelay=true)
   &SpawnRate(SpawnRate=20, Use Spawn Probability=false)
  ]
  particleSpawnScript [
   &InitializeParticle(
    Color=<rgba 1 1 1 0.3>
    Lifetime=&UniformRangedFloat(Minimum=0.4, Maximum=0.6)
    Mass=&UniformRangedFloat(Minimum=0.4, Maximum=2)
    Module.Write Color=true
    Module.Write Lifetime=true
    Module.Write Mass=false
    Module.Write Material Random=true
    Module.Write Position=false
    Module.Write Scale=false
    Module.Write SpriteRotation=true
    Module.Write SpriteSize=true
    Position=&SimulationPosition()
    Sprite Rotation=&NormalizedAngleToDegrees(NormalizedAngle=&UniformRangedFloat(Minimum=0, Maximum=1))
    Sprite Size=&Vector2DFromFloat(Value=&MakeFloatFromVector2d(Vector 2D=&UniformRangedVector2D(Minimum=<1 0>, Maximum=<2 0>)))
   )
   &CylinderLocation(
    Cylinder Height=1
    Cylinder Radius=20
    Lathe Profile=DataInterface() {propertyValues {Curve {keys []}}}
    MaskSpawnGroup=false
    Module.Enable Lathe Profile=false
    Module.Override Randomness=false
    Module.Surface Only=true
    Surface Only Band Thickness=0
    Use Endcaps In Surface Only Mode=false
   ) {enabled=false}
   &SampleSkeletalMeshSkeleton(
    Mesh=DataInterface() {propertyValues {
      DefaultMesh=null
      ExcludeBoneName=root
      OptionalComponentClass=null
      PreviewMesh=&Watering_Can_Strawberry_Gadget
      SkinningMode=SkinOnTheFly
      SoftSourceActor=null
      SourceMode=Default
      UvSetIndex=0
      WholeMeshLOD=-1
      bExcludeBone=false
      bReadDeformedGeometry=true
      bRequireCurrentFrameData=true
      ComponentTags []
      FilteredBones []
      FilteredSockets [FX_Socket]
      MeshUserParameter {parameter {name=None, typeDefHandle {registeredTypeIndex=1}, varData []}}
      SamplingRegions []
     }}
    Spawn Mode:ENiagaraSkeletalSpawning=::NewEnumerator1
   ) {enabled=false}
   &SkeletalMeshSkeletonLocation(Module.Inherit Velocity=false, Module.Use Sampled Bone as Ribbon ID=true) {enabled=false}
   &SystemLocation(MaskSpawnGroup=false, Offset=<-15 100 0>, Offset Coordinate Space=::Simulation)
   &SphereLocation(MaskSpawnGroup=false, Module.Override Randomness=false, Non Uniform Scale=<1 1 0.1>, Sphere Radius=10, Surface Only=false)
   &AddVelocity(Coordinate Space=::Simulation, Velocity=&UniformRangedVector(Minimum=<-75 -75 75>, Maximum=<75 75 100>)) {module=AddVelocity001}
   &ApplyInitialForces()
   &CameraOffset(Camera Offset Amount=5) {enabled=false}
  ]
  particleUpdateScript [
   &ParticleState()
   &GravityForce(Gravity=&UniformRangedVector(Minimum=<0 0 -200>, Maximum=<1 1 -400>))
   &SolveForcesAndVelocity(Clamp Velocity=false, Limit Acceleration=false)
   &ScaleSpriteSizeBySpeed(Max Scale Factor=<1 4>)
  ]
  rendererClasses [&NiagaraSpriteRendererProperties]
  renderers [{rendererClass=&NiagaraSpriteRendererProperties}]
 }
 Water_Pour {
  emitterUpdateScript [
   &EmitterState(Life Cycle Mode=::NewEnumerator1, Loop Behavior=::NewEnumerator1, Loop Delay=1.2, Loop Duration=1.55, UseLoopDelay=true)
   &SpawnRate(SpawnRate=80, Use Spawn Probability=false)
  ]
  particleSpawnScript [
   &InitializeParticle(
    Color=<rgba 1 1 1 0.349>
    Lifetime=&UniformRangedFloat(Minimum=0.6, Maximum=0.8)
    Mass=&UniformRangedFloat(Minimum=0.4, Maximum=2)
    Module.Write Color=true
    Module.Write Lifetime=true
    Module.Write Mass=false
    Module.Write Material Random=true
    Module.Write Position=false
    Module.Write Scale=false
    Module.Write SpriteRotation=true
    Module.Write SpriteSize=true
    Position=&SimulationPosition()
    Sprite Rotation=&NormalizedAngleToDegrees(NormalizedAngle=&UniformRangedFloat(Minimum=0, Maximum=1))
    Sprite Size=&Vector2DFromFloat(Value=&MakeFloatFromVector2d(Vector 2D=&UniformRangedVector2D(Minimum=<2 0>, Maximum=<4 0>)))
   )
   &SampleSkeletalMeshSkeleton(
    Mesh=DataInterface() {propertyValues {
      DefaultMesh=null
      ExcludeBoneName=root
      OptionalComponentClass=null
      PreviewMesh=&Watering_Can_Strawberry_Gadget
      SkinningMode=SkinOnTheFly
      SoftSourceActor=null
      SourceMode=Default
      UvSetIndex=0
      WholeMeshLOD=-1
      bExcludeBone=false
      bReadDeformedGeometry=true
      bRequireCurrentFrameData=true
      ComponentTags []
      FilteredBones []
      FilteredSockets [FX_Socket_Can]
      MeshUserParameter {parameter {name=None, typeDefHandle {registeredTypeIndex=1}, varData []}}
      SamplingRegions []
     }}
    Spawn Mode:ENiagaraSkeletalSpawning=::NewEnumerator1
   )
   &SkeletalMeshSkeletonLocation(Module.Inherit Velocity=false, Module.Use Sampled Bone as Ribbon ID=true)
   &SphereLocation(MaskSpawnGroup=false, Module.Override Randomness=false, Sphere Radius=5, Surface Only=false)
   &AddVelocity(Velocity=&UniformRangedVector(Minimum=<-20 30 0>, Maximum=<20 70 0>))
  ]
  particleUpdateScript [
   &ParticleState()
   &GravityForce()
   &SolveForcesAndVelocity(Clamp Velocity=false, Limit Acceleration=false)
   &ScaleSpriteSizeBySpeed(Max Scale Factor=<1 4>)
   &ScaleSpriteSize(Scale Factor=&Vector2DFromCurve(Vector2Curve=DataInterface() {propertyValues {XCurve {keys [<at 0 0.5>,<at 1 1>]}, YCurve {keys [<at 0 0.5>,<at 1 1>]}}}, ScaleCurve=<1 1>))
  ]
  rendererClasses [&NiagaraSpriteRendererProperties]
  renderers [{rendererClass=&NiagaraSpriteRendererProperties}]
 }
}
summary {
 systemName=NS_Loofah_Emote_StrawberrySparkles
 emitters [
  {emitterName=Water_Burst, rendererClasses [&NiagaraSpriteRendererProperties]}
  {emitterName=Water_Bounce, rendererClasses [&NiagaraSpriteRendererProperties]}
  {emitterName=Water_Pour, rendererClasses [&NiagaraSpriteRendererProperties]}
  {emitterName=PuffSmoke_2, rendererClasses [&NiagaraSpriteRendererProperties]}
  {emitterName=Dirt_Chunks, rendererClasses [&NiagaraSpriteRendererProperties]}
  {emitterName=Sparkles1, rendererClasses [&NiagaraSpriteRendererProperties]}
  {emitterName=Flare, rendererClasses [&NiagaraSpriteRendererProperties, &NiagaraLightRendererProperties]}
  {emitterName=Corona, rendererClasses [&NiagaraSpriteRendererProperties]}
  {emitterName=Sparkles_End, rendererClasses [&NiagaraSpriteRendererProperties]}
  {emitterName=Flare_end, rendererClasses [&NiagaraSpriteRendererProperties, &NiagaraLightRendererProperties]}
  {emitterName=Sparkles_UpwardMotion, rendererClasses [&NiagaraSpriteRendererProperties]}
  {emitterName=PuffSmoke_1, rendererClasses [&NiagaraSpriteRendererProperties]}
  {emitterName=Water_End, rendererClasses [&NiagaraSpriteRendererProperties]}
  {emitterName=Water_Burst001, rendererClasses [&NiagaraSpriteRendererProperties]}
  {emitterName=Flare_Splash_End, rendererClasses [&NiagaraSpriteRendererProperties]}
  {emitterName=Corona_End, rendererClasses [&NiagaraSpriteRendererProperties]}
  {emitterName=Flare_Big_End, rendererClasses [&NiagaraSpriteRendererProperties]}
  {emitterName=Water_GroundSplashes, rendererClasses [&NiagaraSpriteRendererProperties]}
  {emitterName=Flare_Splash_Pour, rendererClasses [&NiagaraSpriteRendererProperties]}
  {
   emitterName=Flare_Light_End
   rendererClasses [&NiagaraSpriteRendererProperties, &NiagaraLightRendererProperties]
  }
 ]
 userVariables []
}
